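# Translucent unlit particle material for sprites and meshes, named as a drone flare refraction effect under the SpawnTeleporter adventure.
# It is set to index-of-refraction distortion and uses a sphere texture as a mask.
# The mask times particle alpha drives a lerp from 1 to 0, so particle fade controls that value.
# The same mask scales a water normal map, confining the ripple detail to the sphere shape.
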Material /Figment_S03/Blueprints/Adventures/SpawnTeleporter/P_DroneFlareRefract
graph {
 properties {
  bUsedWithMeshParticles=true
  bUsedWithParticleSprites=true
  blendMode=BLEND_Translucent
  pixelDepthOffsetMode=PDOM_Legacy
  refractionMethod=RM_IndexOfRefraction
  shadingModel=MSM_Unlit
  usedShadingModels=Unlit
 }
 nodes [
  Constant()
  n1 = TextureSample() {texture=/Game/Effects/Fort_Effects/Textures/Mask/T_Sphere_A}
  LinearInterpolate(Alpha=Multiply(A=ParticleColor().A, B=@n1.R)) {constA=1, constB=0}
  Multiply(A=TextureSample() {samplerType=SAMPLERTYPE_Normal, texture=/Game/Packages/Fortnite_UniqueMaterials_1/Textures/water2_NRM}.RGB, B=@n1.R)
 ]
}
